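# Physics asset for the belt on the female medium Snow Ninja body, defining physics bodies for its dynamic belt bones.
# Six box bodies hang from the pelvis through belt_01 and belt_02, then fork into two two-body strands, 03–04 and 05–06.
# Belt joints share limited swing and tighter twist.
# Capsules on the pelvis and both thighs give the body's shape, with thigh-to-pelvis rotation left free.
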PhysicsAsset /BRCosmetics/Characters/Player/Female/Medium/Bodies/F_MED_Snow_Ninja/Meshes/F_MED_SnowNinja_Belt_Physics
bodies {
 dyn_belt_01 [{shapeType=Box, extentX=3.8846, extentY=2.0097, extentZ=7.5739, center=<1.9752 0.1881 -0.6766>}]
 dyn_belt_02 [{shapeType=Box, extentX=3.8063, extentY=2.0453, extentZ=9.6526, center=<2.7148 0.2088 -0.6683>}]
 dyn_belt_03 [{shapeType=Box, extentX=11.4191, extentY=1.8597, extentZ=6.1968, center=<6.1782 0.0859 -0.0366>}]
 dyn_belt_04 [{shapeType=Box, extentX=9.7463, extentY=2.2831, extentZ=7.2045, center=<3.4863 -0.0318 -0.1736>}]
 dyn_belt_05 [{shapeType=Box, extentX=8.6419, extentY=2.1128, extentZ=5.6962, center=<4.5691 -0.1142 -0.0553>}]
 dyn_belt_06 [{shapeType=Box, extentX=9.8462, extentY=2.5824, extentZ=6.656, center=<3.3945 0.0981 0.2679>}]
 pelvis [{shapeType=Capsule, length=12.3067, radius=12.2622, center=<-8.3377 -1.3502 0.6397>, rotation=<rot 0 -17.0322 -168.7098>}]
 thigh_l [{shapeType=Capsule, length=25.0054, radius=9.3963, center=<-23.327 0.2727 -0.9847>, rotation=<rot 90 0 0>}]
 thigh_r [{shapeType=Capsule, length=24.8323, radius=9.3963, center=<23.1128 -0.2723 0.0178>, rotation=<rot 90 0 0>}]
}
constraints [
 {bone1Name=dyn_belt_01, bone2Name=pelvis, swing1LimitDegrees=45, swing2LimitDegrees=45, twistLimitDegrees=15}
 {bone1Name=dyn_belt_02, bone2Name=dyn_belt_01, swing1LimitDegrees=45, swing2LimitDegrees=45, twistLimitDegrees=15}
 {bone1Name=dyn_belt_03, bone2Name=dyn_belt_02, swing1LimitDegrees=45, swing2LimitDegrees=45, twistLimitDegrees=15}
 {bone1Name=dyn_belt_04, bone2Name=dyn_belt_03, swing1LimitDegrees=45, swing2LimitDegrees=45, twistLimitDegrees=15}
 {bone1Name=dyn_belt_05, bone2Name=dyn_belt_02, swing1LimitDegrees=45, swing2LimitDegrees=45, twistLimitDegrees=15}
 {bone1Name=dyn_belt_06, bone2Name=dyn_belt_05, swing1LimitDegrees=45, swing2LimitDegrees=45, twistLimitDegrees=15}
 {
  bone1Name=thigh_l
  bone2Name=pelvis
  swing1LimitDegrees=45
  swing1Motion=Free
  swing2LimitDegrees=45
  swing2Motion=Free
  twistLimitDegrees=45
  twistMotion=Free
 }
 {
  bone1Name=thigh_r
  bone2Name=pelvis
  swing1LimitDegrees=45
  swing1Motion=Free
  swing2LimitDegrees=45
  swing2Motion=Free
  twistLimitDegrees=45
  twistMotion=Free
 }
]
