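# Physics asset for the female medium Snow Ninja cosmetic body, covering only the pelvis, thighs and dynamic skirt bones.
# Seven three-bone box chains hang from the pelvis with tight swing and twist limits, alongside capsules on the pelvis and both thighs.
# Cross-constraints tie neighboring chains into back, left-front and right-front panels.
# The back-to-right-back links are looser than the other cross-links, and several side-to-front links are duplicated.
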PhysicsAsset /BRCosmetics/Characters/Player/Female/Medium/Bodies/F_MED_Snow_Ninja/Meshes/F_MED_Snow_Ninja_Physics
bodies {
 dyn_skirt_b_01 [{shapeType=Box, extentX=2, extentY=1.5, extentZ=10, center=<1 0 0>}]
 dyn_skirt_b_02 [{shapeType=Box, extentX=2, extentY=1.5, extentZ=10, center=<1 0 0>}]
 dyn_skirt_b_03 [{shapeType=Box, extentX=2, extentY=1.5, extentZ=10, center=<1 0 0>}]
 dyn_skirt_l_01 [{shapeType=Box, extentX=2, extentY=2, extentZ=10, center=<2 0 0>}]
 dyn_skirt_l_02 [{shapeType=Box, extentX=2, extentY=2, extentZ=10, center=<2 0 0>}]
 dyn_skirt_l_03 [{shapeType=Box, extentX=2, extentY=2, extentZ=10, center=<2 0 0>}]
 dyn_skirt_lb_01 [{shapeType=Box, extentX=2, extentY=1.5, extentZ=10, center=<1 0 0>}]
 dyn_skirt_lb_02 [{shapeType=Box, extentX=2, extentY=1.5, extentZ=10, center=<1 0 0>}]
 dyn_skirt_lb_03 [{shapeType=Box, extentX=2, extentY=1.5, extentZ=10, center=<1 0 0>}]
 dyn_skirt_lf_01 [{shapeType=Box, extentX=2, extentY=1.5, extentZ=10, center=<1 0 0>}]
 dyn_skirt_lf_02 [{shapeType=Box, extentX=2, extentY=1.5, extentZ=10, center=<1 0 0>}]
 dyn_skirt_lf_03 [{shapeType=Box, extentX=2, extentY=1.5, extentZ=10.4117, center=<1 0 0>}]
 dyn_skirt_r_01 [{shapeType=Box, extentX=2, extentY=2, extentZ=10, center=<2 0 0>}]
 dyn_skirt_r_02 [{shapeType=Box, extentX=2, extentY=2, extentZ=10, center=<2 0 0>}]
 dyn_skirt_r_03 [{shapeType=Box, extentX=2, extentY=2, extentZ=10, center=<2 0 0>}]
 dyn_skirt_rb_01 [{shapeType=Box, extentX=2, extentY=1.5, extentZ=10, center=<1 0 0>}]
 dyn_skirt_rb_02 [{shapeType=Box, extentX=2, extentY=1.5, extentZ=10, center=<1 0 0>}]
 dyn_skirt_rb_03 [{shapeType=Box, extentX=2, extentY=1.5, extentZ=10, center=<1 0 0>}]
 dyn_skirt_rf_01 [{shapeType=Box, extentX=2, extentY=1.5, extentZ=10, center=<1 0 0>}]
 dyn_skirt_rf_02 [{shapeType=Box, extentX=2, extentY=1.5, extentZ=10, center=<1 0 0>}]
 dyn_skirt_rf_03 [{shapeType=Box, extentX=2, extentY=1.5, extentZ=10, center=<1 0 0>}]
 pelvis [{shapeType=Capsule, length=9.9568, radius=8.01, center=<-2.5187 -3.1243 0.0001>}]
 thigh_l [{shapeType=Capsule, length=19.9786, radius=8.598, center=<-16.2426 0.2722 0.8008>, rotation=<rot 90 0 0>}]
 thigh_r [{shapeType=Capsule, length=18.681, radius=8.9546, center=<15.5606 -0.2723 -0.6849>, rotation=<rot 90 0 0>}]
}
constraints [
 {bone1Name=dyn_skirt_b_01, bone2Name=pelvis, swing1LimitDegrees=20, swing2LimitDegrees=20, twistLimitDegrees=5}
 {bone1Name=dyn_skirt_b_02, bone2Name=dyn_skirt_b_01, swing1LimitDegrees=20, swing2LimitDegrees=20, twistLimitDegrees=2}
 {bone1Name=dyn_skirt_b_03, bone2Name=dyn_skirt_b_02, swing1LimitDegrees=20, swing2LimitDegrees=20, twistLimitDegrees=2}
 {bone1Name=dyn_skirt_lb_01, bone2Name=pelvis, swing1LimitDegrees=20, swing2LimitDegrees=20, twistLimitDegrees=5}
 {bone1Name=dyn_skirt_lb_02, bone2Name=dyn_skirt_lb_01, swing1LimitDegrees=20, swing2LimitDegrees=20, twistLimitDegrees=2}
 {bone1Name=dyn_skirt_lb_03, bone2Name=dyn_skirt_lb_02, swing1LimitDegrees=20, swing2LimitDegrees=20, twistLimitDegrees=2}
 {bone1Name=dyn_skirt_l_01, bone2Name=pelvis, swing1LimitDegrees=20, swing2LimitDegrees=20, twistLimitDegrees=5}
 {bone1Name=dyn_skirt_l_02, bone2Name=dyn_skirt_l_01, swing1LimitDegrees=20, swing2LimitDegrees=20, twistLimitDegrees=2}
 {bone1Name=dyn_skirt_l_03, bone2Name=dyn_skirt_l_02, swing1LimitDegrees=20, swing2LimitDegrees=20, twistLimitDegrees=2}
 {bone1Name=dyn_skirt_lf_01, bone2Name=pelvis, swing1LimitDegrees=20, swing2LimitDegrees=20, twistLimitDegrees=5}
 {bone1Name=dyn_skirt_lf_02, bone2Name=dyn_skirt_lf_01, swing1LimitDegrees=20, swing2LimitDegrees=20, twistLimitDegrees=2}
 {bone1Name=dyn_skirt_lf_03, bone2Name=dyn_skirt_lf_02, swing1LimitDegrees=20, swing2LimitDegrees=20, twistLimitDegrees=2}
 {bone1Name=dyn_skirt_rf_01, bone2Name=pelvis, swing1LimitDegrees=20, swing2LimitDegrees=20, twistLimitDegrees=5}
 {bone1Name=dyn_skirt_rf_02, bone2Name=dyn_skirt_rf_01, swing1LimitDegrees=20, swing2LimitDegrees=20, twistLimitDegrees=2}
 {bone1Name=dyn_skirt_rf_03, bone2Name=dyn_skirt_rf_02, swing1LimitDegrees=20, swing2LimitDegrees=20, twistLimitDegrees=2}
 {bone1Name=dyn_skirt_r_01, bone2Name=pelvis, swing1LimitDegrees=20, swing2LimitDegrees=20, twistLimitDegrees=5}
 {bone1Name=dyn_skirt_r_02, bone2Name=dyn_skirt_r_01, swing1LimitDegrees=20, swing2LimitDegrees=20, twistLimitDegrees=2}
 {bone1Name=dyn_skirt_r_03, bone2Name=dyn_skirt_r_02, swing1LimitDegrees=20, swing2LimitDegrees=20, twistLimitDegrees=2}
 {bone1Name=dyn_skirt_rb_01, bone2Name=pelvis, swing1LimitDegrees=20, swing2LimitDegrees=20, twistLimitDegrees=5}
 {bone1Name=dyn_skirt_rb_02, bone2Name=dyn_skirt_rb_01, swing1LimitDegrees=20, swing2LimitDegrees=20, twistLimitDegrees=2}
 {bone1Name=dyn_skirt_rb_03, bone2Name=dyn_skirt_rb_02, swing1LimitDegrees=20, swing2LimitDegrees=20, twistLimitDegrees=2}
 {bone1Name=thigh_r, bone2Name=pelvis, swing1LimitDegrees=45, swing2LimitDegrees=45, twistLimitDegrees=15}
 {bone1Name=thigh_l, bone2Name=pelvis, swing1LimitDegrees=45, swing2LimitDegrees=45, twistLimitDegrees=15}
 {bone1Name=dyn_skirt_b_03, bone2Name=dyn_skirt_rb_03, swing1LimitDegrees=45, swing2LimitDegrees=45, twistLimitDegrees=15}
 {bone1Name=dyn_skirt_b_02, bone2Name=dyn_skirt_rb_02, swing1LimitDegrees=45, swing2LimitDegrees=45, twistLimitDegrees=15}
 {bone1Name=dyn_skirt_b_01, bone2Name=dyn_skirt_rb_01, swing1LimitDegrees=45, swing2LimitDegrees=45, twistLimitDegrees=15}
 {bone1Name=dyn_skirt_b_03, bone2Name=dyn_skirt_lb_03, swing1LimitDegrees=20, swing2LimitDegrees=20, twistLimitDegrees=2}
 {bone1Name=dyn_skirt_b_02, bone2Name=dyn_skirt_lb_02, swing1LimitDegrees=20, swing2LimitDegrees=20, twistLimitDegrees=2}
 {bone1Name=dyn_skirt_b_01, bone2Name=dyn_skirt_lb_01, swing1LimitDegrees=20, swing2LimitDegrees=20, twistLimitDegrees=2}
 {bone1Name=dyn_skirt_l_03, bone2Name=dyn_skirt_lf_03, swing1LimitDegrees=20, swing2LimitDegrees=20, twistLimitDegrees=2}
 {bone1Name=dyn_skirt_l_02, bone2Name=dyn_skirt_lf_02, swing1LimitDegrees=20, swing2LimitDegrees=20, twistLimitDegrees=2}
 {bone1Name=dyn_skirt_l_01, bone2Name=dyn_skirt_lf_01, swing1LimitDegrees=20, swing2LimitDegrees=20, twistLimitDegrees=2}
 {bone1Name=dyn_skirt_r_02, bone2Name=dyn_skirt_rf_02, swing1LimitDegrees=20, swing2LimitDegrees=20, twistLimitDegrees=2}
 {bone1Name=dyn_skirt_r_01, bone2Name=dyn_skirt_rf_01, swing1LimitDegrees=20, swing2LimitDegrees=20, twistLimitDegrees=2}
 {bone1Name=dyn_skirt_r_03, bone2Name=dyn_skirt_rf_03, swing1LimitDegrees=20, swing2LimitDegrees=20, twistLimitDegrees=2}
 {bone1Name=dyn_skirt_r_02, bone2Name=dyn_skirt_rf_02, swing1LimitDegrees=20, swing2LimitDegrees=20, twistLimitDegrees=2}
 {bone1Name=dyn_skirt_r_01, bone2Name=dyn_skirt_rf_01, swing1LimitDegrees=20, swing2LimitDegrees=20, twistLimitDegrees=2}
 {bone1Name=dyn_skirt_l_03, bone2Name=dyn_skirt_lf_03, swing1LimitDegrees=20, swing2LimitDegrees=20, twistLimitDegrees=2}
 {bone1Name=dyn_skirt_l_02, bone2Name=dyn_skirt_lf_02, swing1LimitDegrees=20, swing2LimitDegrees=20, twistLimitDegrees=2}
 {bone1Name=dyn_skirt_l_01, bone2Name=dyn_skirt_lf_01, swing1LimitDegrees=20, swing2LimitDegrees=20, twistLimitDegrees=2}
]
