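# Post-copy Animation Blueprint for the female medium Snow Ninja body, adding secondary motion to skirt, belt and calf ties.
# It copies the parent mesh pose, then chains RigidBody simulations, thigh constraints and per-bone AnimDynamics.
# Crouching tilts the belt, and targeting enables a right shoulder-pad rotation constraint.
# The event graph derives a head-to-pelvis force per skydive mode, applied to the skirt only while skydiving.
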AnimBlueprint /BRCosmetics/Characters/Player/Female/Medium/Bodies/F_MED_Snow_Ninja/Meshes/F_MED_Snow_Ninja_Skeleton_AnimBP
target_skeleton=&/F_MED_Snow_Ninja_Skeleton
graphs [
 AnimGraph {nodes [
   "Transform (Modify) Bone - Bone: dyn_skirt_fr_l_01"(Alpha="To Float (Boolean)"(InBool=Get bIsCrouching()), Translation="6.893,4.159,-3.475", Rotation="-2.932,-7.714,0.175", Scale="1,1,1") {
    rotationMode=BMM_Additive
    rotationSpace=BCS_BoneSpace
    translationMode=BMM_Additive
    translationSpace=BCS_BoneSpace
    alphaBoolBlend {blendOption=Linear}
    boneToModify {boneName=dyn_skirt_fr_l_01}
    rotation=<rot -2.932 -7.7142 0.175>
    translation=<6.893 4.159 -3.475>
   }
   RigidBody() {
    bOverrideWorldGravity=true
    overridePhysicsAsset=&/F_MED_Snow_Ninja_Physics
    simulationSpace=BaseBoneSpace
    alphaBoolBlend {blendOption=Linear}
    baseBoneRef {boneName=root}
    componentAppliedLinearAccClamp=<1000 1000 1000>
    componentLinearAccScale=<1 1 1>
    componentLinearVelScale=<0.5 0.5 0.5>
    overrideWorldGravity {z=-980}
   }
   n1 = SelectVector(A=Get skydiving vector force() -> (skydiving vector force:Vector), B="0, 0, 0", bPickA=Get bIsSkyDiving())
   n2 = RigidBody(ComponentPose=Local To Component(LocalPose=Copy Pose From Mesh() {bUseAttachedParent=true}), ExternalForce=@n1) {
    bOverrideWorldGravity=true
    overridePhysicsAsset=&/F_MED_SnowNinja_Skirt
    simulationSpace=BaseBoneSpace
    alphaBoolBlend {blendOption=Linear}
    baseBoneRef {boneName=pelvis}
    componentAppliedLinearAccClamp=<5000 5000 5000>
    componentLinearAccScale=<0.5 0.5 0.5>
    componentLinearVelScale=<0.5 0.5 0.5>
    overrideWorldGravity {z=-980}
   }
   n3 = "Constraint - dyn_skirt_fr_l_01 (P)"!(ComponentPose=@n2, Alpha=1) -> (Pose) {
    alphaBoolBlend {blendOption=Linear}
    boneToModify {boneName=dyn_skirt_fr_l_01}
    constraintSetup [{targetBone {boneName=thigh_l}}]
    constraintWeights [1]
   }
   n4 = "Constraint - dyn_skirt_fr_r_01 (P)"!(ComponentPose=@n3, Alpha=1) -> (Pose) {
    alphaBoolBlend {blendOption=Linear}
    boneToModify {boneName=dyn_skirt_fr_r_01}
    constraintSetup [{targetBone {boneName=thigh_r}}]
    constraintWeights [1]
   }
   n5 = "Constraint - dyn_skirt_out_l_01 (P)"!(ComponentPose=@n4, Alpha=0) -> (Pose) {
    alphaBoolBlend {blendOption=Linear}
    boneToModify {boneName=dyn_skirt_out_l_01}
    constraintSetup [{targetBone {boneName=thigh_l}}]
    constraintWeights [1]
   }
   n6 = "Constraint - dyn_skirt_out_r_01 (P)"!(ComponentPose=@n5, Alpha=0) -> (Pose) {
    alphaBoolBlend {blendOption=Linear}
    boneToModify {boneName=dyn_skirt_out_r_01}
    constraintSetup [{targetBone {boneName=thigh_r}}]
    constraintWeights [1]
   }
   n7 = Anim Dynamics - Bone: dyn_skirt_fr_l_01(ComponentPose=@n6, Alpha=1, ExternalForce=@n1) {
    angularDampingOverride=0.98
    bGravityOverrideInSimSpace=true
    bOverrideAngularDamping=true
    bUseGravityOverride=true
    numSolverIterationsPostUpdate=2
    numSolverIterationsPreUpdate=8
    simulationSpace=BoneRelative
    alphaBoolBlend {blendOption=Linear}
    boundBone {boneName=dyn_skirt_fr_l_01}
    gravityOverride {z=-980}
    physicsBodyDefinitions [{
      sphereCollisionRadius=0
      boundBone {boneName=dyn_skirt_fr_l_01}
      boxExtents=<20 20 10>
      constraintSetup {twistAxis=AxisZ, angularLimitsMax=<0 0 2>, angularLimitsMin=<0 -30 -2>, angularTarget=<0 0 0>}
      localJointOffset=<0 0 -3>
     }]
    relativeSpaceBone {boneName=root}
   }
   n8 = Anim Dynamics - Bone: dyn_skirt_fr_l_02(ComponentPose=@n7, Alpha=1, ExternalForce=@n1) {
    angularDampingOverride=0.98
    bGravityOverrideInSimSpace=true
    bOverrideAngularDamping=true
    bUseGravityOverride=true
    simulationSpace=BoneRelative
    alphaBoolBlend {blendOption=Linear}
    boundBone {boneName=dyn_skirt_fr_l_02}
    gravityOverride {z=-980}
    physicsBodyDefinitions [{
      sphereCollisionRadius=0
      boundBone {boneName=dyn_skirt_fr_l_02}
      boxExtents=<20 20 10>
      constraintSetup {twistAxis=AxisZ, angularLimitsMax=<10 1 2>, angularLimitsMin=<-10 -30 -2>, angularTarget=<0 0 0>}
      localJointOffset=<0 0 -3>
     }]
    relativeSpaceBone {boneName=root}
   }
   n9 = Anim Dynamics - Bone: dyn_skirt_fr_l_03(ComponentPose=@n8, Alpha=1, ExternalForce=@n1) {
    angularDampingOverride=0.98
    bGravityOverrideInSimSpace=true
    bOverrideAngularDamping=true
    bUseGravityOverride=true
    simulationSpace=BoneRelative
    alphaBoolBlend {blendOption=Linear}
    boundBone {boneName=dyn_skirt_fr_l_03}
    gravityOverride {z=-980}
    physicsBodyDefinitions [{
      sphereCollisionRadius=0
      boundBone {boneName=dyn_skirt_fr_l_03}
      boxExtents=<20 20 10>
      constraintSetup {twistAxis=AxisZ, angularLimitsMax=<10 0 2>, angularLimitsMin=<-10 -30 -2>, angularTarget=<0 0 0>}
      localJointOffset=<0 0 -3>
     }]
    relativeSpaceBone {boneName=root}
   }
   n10 = Anim Dynamics - Bone: dyn_skirt_fr_r_01(ComponentPose=@n9, Alpha=1, ExternalForce=@n1) {
    angularDampingOverride=0.98
    bGravityOverrideInSimSpace=true
    bOverrideAngularDamping=true
    bUseGravityOverride=true
    simulationSpace=BoneRelative
    alphaBoolBlend {blendOption=Linear}
    boundBone {boneName=dyn_skirt_fr_r_01}
    gravityOverride {z=-980}
    physicsBodyDefinitions [{
      sphereCollisionRadius=0
      boundBone {boneName=dyn_skirt_fr_r_01}
      boxExtents=<20 20 10>
      constraintSetup {twistAxis=AxisZ, angularLimitsMax=<0 0 2>, angularLimitsMin=<0 -30 -2>, angularTarget=<0 0 0>}
      localJointOffset=<0 0 3>
     }]
    relativeSpaceBone {boneName=root}
   }
   n11 = Anim Dynamics - Bone: dyn_skirt_fr_r_02(ComponentPose=@n10, Alpha=1, ExternalForce=@n1) {
    angularDampingOverride=0.98
    bGravityOverrideInSimSpace=true
    bOverrideAngularDamping=true
    bUseGravityOverride=true
    simulationSpace=BoneRelative
    alphaBoolBlend {blendOption=Linear}
    boundBone {boneName=dyn_skirt_fr_r_02}
    gravityOverride {z=-980}
    physicsBodyDefinitions [{
      sphereCollisionRadius=0
      boundBone {boneName=dyn_skirt_fr_r_02}
      boxExtents=<20 20 10>
      constraintSetup {twistAxis=AxisZ, angularLimitsMax=<10 1 2>, angularLimitsMin=<-10 -30 -2>, angularTarget=<0 0 0>}
      localJointOffset=<0 0 3>
     }]
    relativeSpaceBone {boneName=root}
   }
   n12 = Anim Dynamics - Bone: dyn_skirt_fr_r_03(ComponentPose=@n11, Alpha=1, ExternalForce=@n1) {
    angularDampingOverride=0.98
    bGravityOverrideInSimSpace=true
    bOverrideAngularDamping=true
    bUseGravityOverride=true
    simulationSpace=BoneRelative
    alphaBoolBlend {blendOption=Linear}
    boundBone {boneName=dyn_skirt_fr_r_03}
    gravityOverride {z=-980}
    physicsBodyDefinitions [{
      sphereCollisionRadius=0
      boundBone {boneName=dyn_skirt_fr_r_03}
      boxExtents=<20 20 10>
      constraintSetup {twistAxis=AxisZ, angularLimitsMax=<10 3 2>, angularLimitsMin=<-10 -30 -2>, angularTarget=<0 0 0>}
      localJointOffset=<0 0 3>
     }]
    relativeSpaceBone {boneName=root}
   }
   n13 = RigidBody(ComponentPose=@n12) {
    bOverrideWorldGravity=true
    overridePhysicsAsset=&/F_MED_SnowNinja_Belt_Physics
    simulationSpace=BaseBoneSpace
    alphaBoolBlend {blendOption=Linear}
    baseBoneRef {boneName=root}
    componentAppliedLinearAccClamp=<500 500 500>
    overrideWorldGravity {z=-980}
   }
   n14 = "Transform (Modify) Bone - Bone: dyn_belt_01"!(ComponentPose=@n13, bAlphaBoolEnabled:bool=Get bIsCrouching(), Translation="0,0,0", Rotation="0,0,-5", Scale="1,1,1") -> (Pose) {alphaInputType=Bool, rotationMode=BMM_Additive, alphaBoolBlend {blendOption=Linear}, boneToModify {boneName=dyn_belt_01}, rotation {roll=-5}}
   n15 = "Constraint - shoulderpad_r (RR)"!(ComponentPose=@n14, bAlphaBoolEnabled:bool=Get bIsTargeting(), ConstraintWeights_0=0.5, ConstraintWeights_1:float=1) -> (Pose) {
    alphaInputType=Bool
    alphaBoolBlend {blendOption=Linear}
    boneToModify {boneName=shoulderpad_r}
    constraintSetup [{transformType=Rotation, targetBone {boneName=upperarm_r}}, {transformType=Rotation, targetBone {boneName=clavicle_r}}]
    constraintWeights [1,1]
   }
   n16 = Anim Dynamics - Bone: dyn_calf_tie_top(ComponentPose=@n15, Alpha=1, ExternalForce="0,0,0") {
    angularDampingOverride=0.97
    bGravityOverrideInSimSpace=true
    bOverrideAngularDamping=true
    bOverrideLinearDamping=true
    gravityScale=0.65
    lODThreshold=1
    linearDampingOverride=0.97
    numSolverIterationsPostUpdate=4
    numSolverIterationsPreUpdate=16
    alphaBoolBlend {blendOption=Linear}
    boundBone {boneName=dyn_calf_tie_top}
    componentLinearAccScale=<1 1 1>
    componentLinearVelScale=<1 1 1>
    physicsBodyDefinitions [{
      sphereCollisionRadius=0
      boundBone {boneName=dyn_calf_tie_top}
      boxExtents=<5 5 5>
      constraintSetup {angularLimitsMax=<0 30 0>, angularLimitsMin=<0 -30 0>, angularTarget=<0 0 0>}
      localJointOffset=<-0.1 0 0>
     }]
   }
   n17 = Anim Dynamics - Bone: dyn_calf_tie_btm(ComponentPose=@n16, Alpha=1, ExternalForce="0,0,0") {
    angularDampingOverride=0.97
    bGravityOverrideInSimSpace=true
    bOverrideAngularDamping=true
    bOverrideLinearDamping=true
    gravityScale=0.65
    lODThreshold=1
    linearDampingOverride=0.97
    numSolverIterationsPostUpdate=4
    numSolverIterationsPreUpdate=16
    alphaBoolBlend {blendOption=Linear}
    boundBone {boneName=dyn_calf_tie_btm}
    componentLinearAccScale=<1 1 1>
    componentLinearVelScale=<1 1 1>
    physicsBodyDefinitions [{
      sphereCollisionRadius=0
      boundBone {boneName=dyn_calf_tie_btm}
      boxExtents=<5 5 5>
      constraintSetup {angularLimitsMax=<0 30 0>, angularLimitsMin=<0 -30 0>, angularTarget=<0 0 0>}
      localJointOffset=<-0.1 0 0>
     }]
   }
   n18 = Anim Dynamics - Bone: dyn_calf_tie_r(ComponentPose=@n17, Alpha=1, ExternalForce="0,0,0") {
    angularDampingOverride=0.97
    bGravityOverrideInSimSpace=true
    bOverrideAngularDamping=true
    bOverrideLinearDamping=true
    gravityScale=0.65
    lODThreshold=1
    linearDampingOverride=0.97
    numSolverIterationsPostUpdate=4
    numSolverIterationsPreUpdate=16
    alphaBoolBlend {blendOption=Linear}
    boundBone {boneName=dyn_calf_tie_r}
    componentLinearAccScale=<1 1 1>
    componentLinearVelScale=<1 1 1>
    physicsBodyDefinitions [{
      sphereCollisionRadius=0
      boundBone {boneName=dyn_calf_tie_r}
      boxExtents=<5 5 5>
      constraintSetup {angularLimitsMax=<0 30 0>, angularLimitsMin=<0 -30 0>, angularTarget=<0 0 0>}
      localJointOffset=<0.1 0 0>
     }]
   }
   Output Pose(Result=Component To Local(ComponentPose=@n18)) {layerGroup=None}
  ]}
 EventGraph {nodes [
   Event BlueprintUpdateAnimation()
   > n1 = Branch(Condition=Get bCanPlayCustomAnimations() -> (bCanPlayCustomAnimations:bool))
   n2 = Branch(execute=[@n1.else, @n1.then], Condition=Get bIsSkyDiving())
   n3 = Branch(execute=[@n2.else, @n2.then], Condition=Get bIsSkydiveFloating())
   n4 = Get Direction Between Sockets!(
    Component:Skeletal Mesh Component Object Reference=GetOwnerSkeletalMeshComponent()
    SocketOrBoneNameFrom:Name=head
    SocketOrBoneNameTo:Name=pelvis
   ) -> (ReturnValue:Vector) {is=title K2Node_CallFunction}
   > Set skydiving vector force(skydiving vector force:Vector=vector + vector(
     A:Vector=vector * vector(A:Vector=@n4, B:Vector="75,75,75") -> (ReturnValue:Vector) {is=title K2Node_CommutativeAssociativeBinaryOperator}
     B:Vector="0, 0,2000"
    ) -> (ReturnValue:Vector) {is=title K2Node_CommutativeAssociativeBinaryOperator}) -> (Output_Get:Vector)
   n5 = Branch(execute=@n3.else, Condition=Get bIsSkydiveDiveMode())
   > Set skydiving vector force(skydiving vector force=vector + vector(A=vector * vector(A=@n4, B="180,180,200"), B="0, 0,2000"))
   Set skydiving vector force(execute=@n5.else, skydiving vector force=vector * vector(A=@n4, B="-15,-15,-15"))
  ]}
]
variables [skydiving vector force {type=Vector}]
